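# MetaSound patch that swaps between surface-type loop sounds, crossfading to NextSoundToPlay on each PlayNextSound trigger.
# A trigger sequence alternates two mono wave players, with interpolated mixer gains fading one in and the other out over CrossfadeTime.
# Delayed triggers stop the faded player and report crossfade completion; FadeOutAllSounds fades both and signals when done.
# The Mobile page uses one player with no crossfade, fade-out or completion triggers.
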MetaSoundPatch /FNGameplayCues/Sounds/Fort_Audio_Blueprints/SurfaceTypeSound/MSP_SurfaceLoopSwapper
inputs [
 CrossfadeTime {type=Time}
 FadeOutAllSounds {type=Trigger}
 NextSoundToPlay {type=WaveAsset}
 PitchShift {type=Float}
 PlayNextSound {type=Trigger}
]
outputs [
 OnCrossfadeFinished {type=Trigger}
 OnFadeOutAllSoundsFinished {type=Trigger}
 Out {type=Audio}
]
pages [
 Default {
  is_build_graph=true
  input_defaults {CrossfadeTime=0.5, FadeOutAllSounds=null, NextSoundToPlay=null, PitchShift=0, PlayNextSound=null}
  nodes [
   n1 = InitVariable.Float()
   n2 = InitVariable.Trigger()
   TriggerSequence2_0 = "UE.Trigger Sequence (2)"(In=^PlayNextSound)
   n3 = VariableAccessor.Trigger(Variable=VariableMutator.Trigger(Variable=InitVariable.Trigger().Variable, Value=@TriggerSequence2_0.Out 1).Variable)
   n4 = VariableAccessor.Trigger(Variable=@n3.Variable)
   TriggerAnyTriggerAny2_2 = "TriggerAny.Trigger Any (2)"(In 0=^FadeOutAllSounds, In 1=VariableAccessor.Trigger(Variable=@n4.Variable).Value)
   n5 = VariableAccessor.Trigger(Variable=VariableMutator.Trigger(Variable=InitVariable.Trigger().Variable, Value=@TriggerSequence2_0.Out 0).Variable)
   n6 = VariableAccessor.Trigger(Variable=@n5.Variable)
   n7 = Value.Value:Float(
    Reset=VariableAccessor.Trigger(Variable=VariableMutator.Trigger(Variable=@n2.Variable, Value=@TriggerAnyTriggerAny2_2.Out).Variable).Value
    Set=@n6.Value
   ) {also {defaults {Target Value=1}}}
   n8 = UE.InterpTo:Audio(Interp Time=^CrossfadeTime, Target=@n7.Output Value)
   n9 = InitVariable.Float()
   n10 = InitVariable.Trigger()
   TriggerAnyTriggerAny2_1 = "TriggerAny.Trigger Any (2)"(In 0=^FadeOutAllSounds, In 1=VariableAccessor.Trigger(Variable=@n6.Variable).Value)
   n11 = Value.Value:Float(
    Set=@n4.Value
    Reset=VariableAccessor.Trigger(Variable=VariableMutator.Trigger(Variable=@n10.Variable, Value=@TriggerAnyTriggerAny2_1.Out).Variable).Value
   ) {also {defaults {Init Value=0, Target Value=1}}}
   n12 = UE.InterpTo:Audio(Interp Time=^CrossfadeTime, Target=@n11.Output Value)
   TriggerDelay_1 = UE.Trigger Delay(Delay Time=^CrossfadeTime, In=@n11.On Reset, Reset=@n11.On Set)
   n13 = VariableAccessor.Trigger(Variable=VariableMutator.Trigger(Variable=InitVariable.Trigger().Variable, Value=@TriggerDelay_1.Out).Variable)
   WavePlayer_0 = UE.Wave Player:Mono(Wave Asset=^NextSoundToPlay, Pitch Shift=^PitchShift, Stop=@n13.Value, Play=@n3.Value) {also {defaults {Loop="<MetaSoundBuilderResult.SUCCEEDED: 0>"}}}
   TriggerDelay_2 = UE.Trigger Delay(Delay Time=^CrossfadeTime, In=@n7.On Reset, Reset=@n7.On Set)
   n14 = VariableAccessor.Trigger(Variable=VariableMutator.Trigger(Variable=InitVariable.Trigger().Variable, Value=@TriggerDelay_2.Out).Variable)
   WavePlayer_1 = UE.Wave Player:Mono(Wave Asset=^NextSoundToPlay, Pitch Shift=^PitchShift, Stop=@n14.Value, Play=@n5.Value) {also {defaults {Loop="<MetaSoundBuilderResult.SUCCEEDED: 0>"}}}
   AudioMixerAudioMixerMono2_0 = "AudioMixer.Audio Mixer (Mono, 2)"(
    Gain 0=VariableAccessor.Float(Variable=VariableMutator.Float(Variable=@n1.Variable, Value=@n8.Value).Variable).Value
    Gain 1=VariableAccessor.Float(Variable=VariableMutator.Float(Variable=@n9.Variable, Value=@n12.Value).Variable).Value
    In 1=@WavePlayer_0.Out Mono
    In 0=@WavePlayer_1.Out Mono
   )
   TriggerAnyTriggerAny2_0 = "TriggerAny.Trigger Any (2)"(In 1=VariableAccessor.Trigger(Variable=@n13.Variable).Value, In 0=VariableAccessor.Trigger(Variable=@n14.Variable).Value)
   TriggerDelay_0 = UE.Trigger Delay(Delay Time=^CrossfadeTime, In=^FadeOutAllSounds)
  ]
  links [
   {dst="@Out", src=AudioMixerAudioMixerMono2_0.Out}
   {dst="@OnCrossfadeFinished", src=TriggerAnyTriggerAny2_0.Out}
   {dst="@OnFadeOutAllSoundsFinished", src=TriggerDelay_0.Out}
  ]
 }
 Mobile {
  nodes [
   InitVariable.Float()
   InitVariable.Float()
   InitVariable.Trigger()
   InitVariable.Trigger()
   InitVariable.Trigger()
   InitVariable.Trigger()
   InitVariable.Trigger()
   InitVariable.Trigger()
   WavePlayer_0 = UE.Wave Player:Mono(Wave Asset=^NextSoundToPlay, Pitch Shift=^PitchShift, Play=^PlayNextSound) {also {defaults {Loop="<MetaSoundBuilderResult.SUCCEEDED: 0>"}}}
  ]
  links [{dst="@Out", src=WavePlayer_0.Out Mono}]
 }
]
